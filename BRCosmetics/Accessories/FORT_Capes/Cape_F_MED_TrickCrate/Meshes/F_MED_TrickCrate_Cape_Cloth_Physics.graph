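# Physics asset for the female medium TrickCrate cape, supplying body collision shapes that, per its name, serve the cape's cloth simulation.
# Capsules cover the pelvis, five spine bones, upper arms, thighs and calves, with notably wide thigh capsules.
# A huge sphere on the root, centered far below it, likely acts as a near-flat floor for the cloth.
# No constraints are defined, so the file provides only collision volumes, with no ragdoll joints.
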PhysicsAsset /BRCosmetics/Accessories/FORT_Capes/Cape_F_MED_TrickCrate/Meshes/F_MED_TrickCrate_Cape_Cloth_Physics
bodies {
 calf_l [{
   shapeType=Capsule
   length=28.4881
   radius=8.7772
   shapeName=calf_l_capsule
   center=<-14.476 -1.9053 0.9405>
   rotation=<rot 0 -90.0001 -86.2828>
  }]
 calf_r [{
   shapeType=Capsule
   length=31.4881
   radius=9.7772
   shapeName=calf_l_capsule
   center=<14.476 1.9053 0.9405>
   rotation=<rot 0 -90.0001 -86.2828>
  }]
 pelvis [{shapeType=Capsule, length=6.9574, radius=14.227, center=<-0.2698 0.1908 0>}]
 root [{shapeType=Sphere, radius=298, shapeName=root_sphere, center=<0 -24.5697 -289.4702>}]
 spine_01 [{shapeType=Capsule, length=7.9438, radius=13.5586, center=<-1.7965 2.2546 0>}]
 spine_02 [{shapeType=Capsule, length=5.7999, radius=10.1068, center=<-1.711 1.6489 0.0001>}]
 spine_03 [{shapeType=Capsule, length=6.6842, radius=11.1776, center=<0.0216 3.4844 0.0001>}]
 spine_04 [{shapeType=Capsule, length=7.9139, radius=11.5355, center=<0.1832 2.4254 0.0002>}]
 spine_05 [{shapeType=Capsule, length=6.0496, radius=12.2199, center=<-0.4501 2.2217 0>}]
 thigh_l [{shapeType=Capsule, length=22.9881, radius=22, shapeName=calf_l_capsule, center=<-14.476 -1.9053 0.9405>, rotation=<rot 0 -90.0005 -86.2828>}]
 thigh_r [{shapeType=Capsule, length=21.4881, radius=22, shapeName=calf_l_capsule, center=<14.476 1.9053 0.9405>, rotation=<rot 0 -90.0006 -86.2828>}]
 upperarm_l [{shapeType=Capsule, length=22.125, radius=6, center=<11.9672 -1.662 0>, rotation=<rot 90 0 -6.5611>}]
 upperarm_r [{shapeType=Capsule, length=23.625, radius=6, center=<-11.9672 1.662 0>, rotation=<rot 90 0 173.4389>}]
}
constraints []
