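# Physics asset for the TrickCrate cape cosmetic on a medium female character, defining capsule bodies and constraints for cape motion.
# Capsules cover torso, legs, upper arms and two hair bones alongside the cape.
# Four limited-angle chains form the cape: outer pairs start at clavicle_cape bones, middle pairs at MSTR bones under backpack_root.
# Angle-free cross-links join neighboring chains at upper and lower segments and both middle chains, making separate strands behave like one sheet.
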PhysicsAsset /BRCosmetics/Accessories/FORT_Capes/Cape_F_MED_TrickCrate/Meshes/F_MED_TrickCrate_Cape_Physics
bodies {
 backpack_root [{shapeType=Capsule, length=1, radius=1}]
 calf_l [{shapeType=Capsule, length=11.7963, radius=39.0787, center=<-26.1258 20 0>, rotation=<rot 90 179.8881 -178.6232>}]
 calf_r [{shapeType=Capsule, length=11.7963, radius=39.0787, center=<26.1258 -20 0>, rotation=<rot 90 180 1.4887>}]
 cape_01_l [{shapeType=Capsule, length=1, radius=1}]
 cape_01_r [{shapeType=Capsule, length=1, radius=1}]
 cape_02_l [{shapeType=Capsule, length=12, radius=5.5, center=<10 0 10>, rotation=<rot -90 0 0>}]
 cape_02_r [{shapeType=Capsule, length=12, radius=5.5, center=<10 0 10>, rotation=<rot -90 0 0>}]
 cape_03_l [{shapeType=Capsule, length=12, radius=8, center=<10 0 10>, rotation=<rot -90 0 0>}]
 cape_03_r [{shapeType=Capsule, length=12, radius=8, center=<10 0 10>, rotation=<rot -90 0 0>}]
 cape_04_l [{shapeType=Capsule, length=12, radius=10, center=<10 0 10>, rotation=<rot -90 0 0>}]
 cape_04_r [{shapeType=Capsule, length=12, radius=10, center=<10 0 10>, rotation=<rot -90 0 0>}]
 cape_05_l [{shapeType=Capsule, length=12, radius=10, center=<10 0 10>, rotation=<rot -90 0 0>}]
 cape_05_r [{shapeType=Capsule, length=12, radius=10, center=<10 0 10>, rotation=<rot -90 0 0>}]
 cape_06_l [{shapeType=Capsule, length=12, radius=10, center=<10 0 10>, rotation=<rot -90 0 0>}]
 cape_06_r [{shapeType=Capsule, length=12, radius=10, center=<10 0 10>, rotation=<rot -90 0 0>}]
 cape_07_l [{shapeType=Capsule, length=12, radius=10, center=<10 0 10>, rotation=<rot -90 0 0>}]
 cape_07_r [{shapeType=Capsule, length=12, radius=10, center=<10 0 10>, rotation=<rot -90 0 0>}]
 cape_mid_01_l [{shapeType=Capsule, length=12, radius=9, center=<9 0 10>, rotation=<rot 90 0 0>}]
 cape_mid_01_l_MSTR [{shapeType=Capsule, length=1, radius=1}]
 cape_mid_01_r [{shapeType=Capsule, length=12, radius=9, center=<9 0 10>, rotation=<rot 90 0 0>}]
 cape_mid_01_r_MSTR [{shapeType=Capsule, length=1, radius=1}]
 cape_mid_02_l [{shapeType=Capsule, length=12, radius=8, center=<10 0 10>, rotation=<rot -90 0 0>}]
 cape_mid_02_r [{shapeType=Capsule, length=12, radius=8, center=<10 0 10>, rotation=<rot -90 0 0>}]
 cape_mid_03_l [{shapeType=Capsule, length=12, radius=8, center=<10 0 10>, rotation=<rot -90 0 0>}]
 cape_mid_03_r [{shapeType=Capsule, length=12, radius=8, center=<10 0 10>, rotation=<rot -90 0 0>}]
 cape_mid_04_l [{shapeType=Capsule, length=12, radius=10, center=<10 0 10>, rotation=<rot -90 0 0>}]
 cape_mid_04_r [{shapeType=Capsule, length=12, radius=10, center=<10 0 10>, rotation=<rot -90 0 0>}]
 cape_mid_05_l [{shapeType=Capsule, length=12, radius=10, center=<10 0 10>, rotation=<rot -90 0 0>}]
 cape_mid_05_r [{shapeType=Capsule, length=12, radius=10, center=<10 0 10>, rotation=<rot -90 0 0>}]
 cape_mid_06_l [{shapeType=Capsule, length=12, radius=10, center=<10 0 10>, rotation=<rot -90 0 0>}]
 cape_mid_06_r [{shapeType=Capsule, length=12, radius=10, center=<10 0 10>, rotation=<rot -90 0 0>}]
 clavicle_cape_l [{shapeType=Capsule, length=1, radius=1}]
 clavicle_cape_r [{shapeType=Capsule, length=1, radius=1}]
 clavicle_l [{shapeType=Capsule, length=15.1649, radius=7.4601, center=<12.3638 -1.6165 0>, rotation=<rot 90 0 -6.5611>}]
 clavicle_r [{shapeType=Capsule, length=15.1649, radius=7.4601, center=<-12.3638 1.6165 0>, rotation=<rot 90 0 173.4389>}]
 dyn_hair_05_cr [{
   shapeType=Capsule
   length=20.5801
   radius=16.7949
   shapeName=dyn_hair_05_cr_capsule
   center=<0.0146 0.052 34.1656>
   rotation=<rot 89.61 -180 -89.6999>
  }]
 dyn_hair_07_cr [{
   shapeType=Capsule
   length=27.0842
   radius=11.5743
   shapeName=dyn_hair_07_cr_capsule
   center=<2.7684 0.0308 27.1587>
   rotation=<rot 89.61 180 -89.6999>
  }]
 pelvis [{shapeType=Capsule, length=6.0938, radius=15, center=<-8.6013 -5.4554 0>}]
 spine_01 [{shapeType=Capsule, length=19.4776, radius=10.5224, shapeName=spine_01_capsule}]
 spine_02 [{shapeType=Capsule, length=8.6403, radius=10.1722, center=<3.5949 0 0>}]
 spine_03 [{shapeType=Capsule, length=42.7679, radius=6.8884, center=<2.5875 -1 0>}]
 spine_04 [{shapeType=Capsule, length=35.2196, radius=8.1866, center=<2.8832 -0.1618 0>}]
 spine_05 [{shapeType=Capsule, length=20.5301, radius=9.4699, shapeName=spine_05_capsule}]
 thigh_l [{shapeType=Capsule, length=28.0938, radius=14, center=<-19.7778 -0.1788 0>, rotation=<rot 90 18.9704 16.3433>}]
 thigh_r [{shapeType=Capsule, length=28.0938, radius=14, center=<19.7778 0.1788 0>, rotation=<rot 90 -10.6197 166.7532>}]
 upperarm_l [{shapeType=Capsule, length=20.625, radius=8, center=<11.9672 -1.662 0>, rotation=<rot 90 0 -6.5611>}]
 upperarm_r [{shapeType=Capsule, length=20.625, radius=8, center=<-11.9672 1.662 0>, rotation=<rot 90 0 173.4389>}]
}
constraints [
 {
  bone1Name=calf_r
  bone2Name=thigh_r
  swing1LimitDegrees=45
  swing1Motion=Free
  swing2LimitDegrees=45
  swing2Motion=Free
  twistLimitDegrees=45
  twistMotion=Free
 }
 {
  bone1Name=spine_01
  bone2Name=pelvis
  swing1LimitDegrees=45
  swing1Motion=Free
  swing2LimitDegrees=45
  swing2Motion=Free
  twistLimitDegrees=45
  twistMotion=Free
 }
 {
  bone1Name=spine_02
  bone2Name=spine_01
  swing1LimitDegrees=45
  swing1Motion=Free
  swing2LimitDegrees=45
  swing2Motion=Free
  twistLimitDegrees=45
  twistMotion=Free
 }
 {
  bone1Name=spine_03
  bone2Name=spine_02
  swing1LimitDegrees=45
  swing1Motion=Free
  swing2LimitDegrees=45
  swing2Motion=Free
  twistLimitDegrees=45
  twistMotion=Free
 }
 {
  bone1Name=spine_04
  bone2Name=spine_03
  swing1LimitDegrees=45
  swing1Motion=Free
  swing2LimitDegrees=45
  swing2Motion=Free
  twistLimitDegrees=45
  twistMotion=Free
 }
 {
  bone1Name=spine_05
  bone2Name=spine_04
  swing1LimitDegrees=45
  swing1Motion=Free
  swing2LimitDegrees=45
  swing2Motion=Free
  twistLimitDegrees=45
  twistMotion=Free
 }
 {
  bone1Name=backpack_root
  bone2Name=spine_05
  swing1LimitDegrees=45
  swing1Motion=Free
  swing2LimitDegrees=45
  swing2Motion=Free
  twistLimitDegrees=45
  twistMotion=Free
 }
 {
  bone1Name=cape_mid_01_l_MSTR
  bone2Name=backpack_root
  swing1LimitDegrees=45
  swing1Motion=Free
  swing2LimitDegrees=45
  swing2Motion=Free
  twistLimitDegrees=45
  twistMotion=Free
 }
 {
  bone1Name=cape_mid_01_r_MSTR
  bone2Name=backpack_root
  swing1LimitDegrees=45
  swing1Motion=Free
  swing2LimitDegrees=45
  swing2Motion=Free
  twistLimitDegrees=45
  twistMotion=Free
 }
 {
  bone1Name=clavicle_l
  bone2Name=spine_05
  swing1LimitDegrees=45
  swing1Motion=Free
  swing2LimitDegrees=45
  swing2Motion=Free
  twistLimitDegrees=45
  twistMotion=Free
 }
 {bone1Name=cape_01_l, bone2Name=clavicle_cape_l, swing1LimitDegrees=45, swing2LimitDegrees=45, twistLimitDegrees=5}
 {
  bone1Name=upperarm_l
  bone2Name=clavicle_l
  swing1LimitDegrees=45
  swing1Motion=Free
  swing2LimitDegrees=45
  swing2Motion=Free
  twistLimitDegrees=45
  twistMotion=Free
 }
 {
  bone1Name=clavicle_r
  bone2Name=spine_05
  swing1LimitDegrees=45
  swing1Motion=Free
  swing2LimitDegrees=45
  swing2Motion=Free
  twistLimitDegrees=45
  twistMotion=Free
 }
 {bone1Name=cape_01_r, bone2Name=clavicle_cape_r, swing1LimitDegrees=45, swing2LimitDegrees=45, twistLimitDegrees=5}
 {
  bone1Name=upperarm_r
  bone2Name=clavicle_r
  swing1LimitDegrees=45
  swing1Motion=Free
  swing2LimitDegrees=45
  swing2Motion=Free
  twistLimitDegrees=45
  twistMotion=Free
 }
 {
  bone1Name=thigh_l
  bone2Name=pelvis
  swing1LimitDegrees=45
  swing1Motion=Free
  swing2LimitDegrees=45
  swing2Motion=Free
  twistLimitDegrees=45
  twistMotion=Free
 }
 {
  bone1Name=calf_l
  bone2Name=thigh_l
  swing1LimitDegrees=45
  swing1Motion=Free
  swing2LimitDegrees=45
  swing2Motion=Free
  twistLimitDegrees=45
  twistMotion=Free
 }
 {
  bone1Name=thigh_r
  bone2Name=pelvis
  swing1LimitDegrees=45
  swing1Motion=Free
  swing2LimitDegrees=45
  swing2Motion=Free
  twistLimitDegrees=45
  twistMotion=Free
 }
 {bone1Name=cape_mid_01_l, bone2Name=cape_mid_01_l_MSTR, swing1LimitDegrees=5, swing2LimitDegrees=45, twistLimitDegrees=5}
 {bone1Name=cape_mid_01_r, bone2Name=cape_mid_01_r_MSTR, swing1LimitDegrees=5, swing2LimitDegrees=45, twistLimitDegrees=5, twistMotion=Locked}
 {bone1Name=cape_mid_02_l, bone2Name=cape_mid_01_l, swing1LimitDegrees=5, swing2LimitDegrees=5, twistLimitDegrees=5}
 {bone1Name=cape_mid_03_l, bone2Name=cape_mid_02_l, swing1LimitDegrees=5, swing2LimitDegrees=45, twistLimitDegrees=5}
 {bone1Name=cape_mid_04_l, bone2Name=cape_mid_03_l, swing1LimitDegrees=5, swing2LimitDegrees=45, twistLimitDegrees=5}
 {bone1Name=cape_mid_05_l, bone2Name=cape_mid_04_l, swing1LimitDegrees=5, swing2LimitDegrees=45, twistLimitDegrees=5}
 {bone1Name=cape_mid_06_l, bone2Name=cape_mid_05_l, swing1LimitDegrees=5, swing2LimitDegrees=45, twistLimitDegrees=5}
 {bone1Name=cape_mid_02_r, bone2Name=cape_mid_01_r, swing1LimitDegrees=5, swing2LimitDegrees=5, twistLimitDegrees=5}
 {bone1Name=cape_mid_03_r, bone2Name=cape_mid_02_r, swing1LimitDegrees=5, swing2LimitDegrees=45, twistLimitDegrees=5}
 {bone1Name=cape_mid_04_r, bone2Name=cape_mid_03_r, swing1LimitDegrees=5, swing2LimitDegrees=45, twistLimitDegrees=5}
 {bone1Name=cape_mid_05_r, bone2Name=cape_mid_04_r, swing1LimitDegrees=5, swing2LimitDegrees=45, twistLimitDegrees=5}
 {bone1Name=cape_mid_06_r, bone2Name=cape_mid_05_r, swing1LimitDegrees=5, swing2LimitDegrees=45, twistLimitDegrees=5}
 {bone1Name=cape_02_l, bone2Name=cape_01_l, swing1LimitDegrees=35, swing2LimitDegrees=45, twistLimitDegrees=5}
 {bone1Name=cape_03_l, bone2Name=cape_02_l, swing1LimitDegrees=5, swing2LimitDegrees=5, twistLimitDegrees=5}
 {bone1Name=cape_04_l, bone2Name=cape_03_l, swing1LimitDegrees=5, swing2LimitDegrees=45, twistLimitDegrees=5}
 {bone1Name=cape_05_l, bone2Name=cape_04_l, swing1LimitDegrees=5, swing2LimitDegrees=45, twistLimitDegrees=5}
 {bone1Name=cape_06_l, bone2Name=cape_05_l, swing1LimitDegrees=25, swing2LimitDegrees=45, twistLimitDegrees=5}
 {bone1Name=cape_07_l, bone2Name=cape_06_l, swing1LimitDegrees=25, swing2LimitDegrees=45, twistLimitDegrees=5}
 {bone1Name=cape_02_r, bone2Name=cape_01_r, swing1LimitDegrees=35, swing2LimitDegrees=45, twistLimitDegrees=5}
 {bone1Name=cape_03_r, bone2Name=cape_02_r, swing1LimitDegrees=5, swing2LimitDegrees=5, twistLimitDegrees=5}
 {bone1Name=cape_04_r, bone2Name=cape_03_r, swing1LimitDegrees=5, swing2LimitDegrees=45, twistLimitDegrees=5}
 {bone1Name=cape_05_r, bone2Name=cape_04_r, swing1LimitDegrees=5, swing2LimitDegrees=45, twistLimitDegrees=5}
 {bone1Name=cape_06_r, bone2Name=cape_05_r, swing1LimitDegrees=25, swing2LimitDegrees=45, twistLimitDegrees=5}
 {bone1Name=cape_07_r, bone2Name=cape_06_r, swing1LimitDegrees=25, swing2LimitDegrees=45, twistLimitDegrees=5}
 {
  bone1Name=cape_06_r
  bone2Name=cape_mid_05_r
  swing1LimitDegrees=45
  swing1Motion=Free
  swing2LimitDegrees=45
  swing2Motion=Free
  twistLimitDegrees=45
  twistMotion=Free
 }
 {
  bone1Name=cape_06_l
  bone2Name=cape_mid_05_l
  swing1LimitDegrees=45
  swing1Motion=Free
  swing2LimitDegrees=45
  swing2Motion=Free
  twistLimitDegrees=45
  twistMotion=Free
 }
 {
  bone1Name=cape_mid_05_r
  bone2Name=cape_mid_05_l
  swing1LimitDegrees=45
  swing1Motion=Free
  swing2LimitDegrees=45
  swing2Motion=Free
  twistLimitDegrees=45
  twistMotion=Free
 }
 {
  bone1Name=cape_mid_05_l
  bone2Name=cape_mid_05_r
  swing1LimitDegrees=45
  swing1Motion=Free
  swing2LimitDegrees=45
  swing2Motion=Free
  twistLimitDegrees=45
  twistMotion=Free
 }
 {
  bone1Name=cape_03_r
  bone2Name=cape_mid_02_r
  swing1LimitDegrees=45
  swing1Motion=Free
  swing2LimitDegrees=45
  swing2Motion=Free
  twistLimitDegrees=45
  twistMotion=Free
 }
 {
  bone1Name=cape_03_l
  bone2Name=cape_mid_02_l
  swing1LimitDegrees=45
  swing1Motion=Free
  swing2LimitDegrees=45
  swing2Motion=Free
  twistLimitDegrees=45
  twistMotion=Free
 }
]
