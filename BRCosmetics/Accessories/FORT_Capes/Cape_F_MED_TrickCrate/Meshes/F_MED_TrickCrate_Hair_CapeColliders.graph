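# Physics asset for the TrickCrate cape cosmetic, named as the hair's cape colliders, likely keeping simulated cloth or hair from clipping.
# It places one capsule on each of two dynamic hair bones, dyn_hair_05_cr and dyn_hair_07_cr.
# A single constraint links the two bodies with swing and twist left free, so its angular limit values restrict nothing.
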PhysicsAsset /BRCosmetics/Accessories/FORT_Capes/Cape_F_MED_TrickCrate/Meshes/F_MED_TrickCrate_Hair_CapeColliders
bodies {
 dyn_hair_05_cr [{
   shapeType=Capsule
   length=20.5801
   radius=16.7949
   shapeName=dyn_hair_05_cr_capsule
   center=<0.137 0.052 16.1903>
   rotation=<rot 89.61 -180 -89.6999>
  }]
 dyn_hair_07_cr [{
   shapeType=Capsule
   length=27.0842
   radius=11.5743
   shapeName=dyn_hair_07_cr_capsule
   center=<2.8826 0.0308 10.3844>
   rotation=<rot 89.61 180 -89.6999>
  }]
}
constraints [{
  bone1Name=dyn_hair_07_cr
  bone2Name=dyn_hair_05_cr
  swing1LimitDegrees=45
  swing1Motion=Free
  swing2LimitDegrees=45
  swing2Motion=Free
  twistLimitDegrees=45
  twistMotion=Free
 }]
